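# Physics asset for the belt mesh of the Weeping Woods Toon medium male player cosmetic body, providing physics for a dangling right-side belt piece.
# Capsules on pelvis and spine_01 approximate the torso, while a long offset capsule covers the dynamic belt bone.
# A negligible-size master body links the belt to spine_01 with free rotation.
# The belt-to-master joint has a single nonzero swing limit, giving hinge-like motion.
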PhysicsAsset /BRCosmetics/Characters/Player/Male/Medium/Bodies/M_MED_Weeping_Woods_Toon/Meshes/M_MED_Weeping_Woods_Toon_belt_Physics
bodies {
 dyn_belt_NEW_01_r [{shapeType=Capsule, length=22.8438, radius=3.4431, shapeName=dyn_belt_NEW_01_r_capsule, center=<14.5568 -4 -5.5>, rotation=<rot -5 0 10>}]
 dyn_belt_NEW_01_r_MSTR [{shapeType=Capsule, length=0.01, radius=0.01}]
 pelvis [{shapeType=Capsule, length=7.1988, radius=14.0546}]
 spine_01 [{shapeType=Capsule, length=19.3211, radius=14.2254, shapeName=spine_01_capsule, center=<4.2799 -1.9834 1.732>, rotation=<rot 90 0 0>}]
}
constraints [
 {
  bone1Name=dyn_belt_NEW_01_r_MSTR
  bone2Name=spine_01
  swing1LimitDegrees=45
  swing1Motion=Free
  swing2LimitDegrees=45
  swing2Motion=Free
  twistLimitDegrees=45
  twistMotion=Free
 }
 {
  bone1Name=spine_01
  bone2Name=pelvis
  swing1LimitDegrees=45
  swing1Motion=Free
  swing2LimitDegrees=45
  swing2Motion=Free
  twistLimitDegrees=45
  twistMotion=Free
 }
 {bone1Name=dyn_belt_NEW_01_r, bone2Name=dyn_belt_NEW_01_r_MSTR, swing1LimitDegrees=45, swing2LimitDegrees=0, twistLimitDegrees=0}
]
